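# Translucent unlit, two-sided Niagara sprite material for the SwoopClasp skydiving contrail, drawing flickering hologram shapes tinted by particle color.
# It picks a 2x2 flipbook cell from a packed mask's blue channel, indexed by dynamic or scalar parameter.
# Time-driven hologram glitch modulates that mask, blending out as AnimBranch rises.
# An optional branch mode thresholds the result against AnimBranch, and a view-angle fade multiplies the final mask.
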Material /BRCosmetics/Athena/Effects/Particles/Skydiving/NS_Contrail_SwoopClasp/M_Contrail_SwoopClasp_Circle
graph {
 properties {
  bAutomaticallySetUsageInEditor=false
  bUsedWithNiagaraSprites=true
  blendMode=BLEND_Translucent
  shadingModel=MSM_Unlit
  twoSided=true
  usedShadingModels=Unlit
 }
 nodes [
  MaterialFunctionCall(Emissive=ParticleColor().RGB) {materialFunction=&MF_FX_EmissiveAdjustments}
  n1 = TextureObject() {samplerType=SAMPLERTYPE_Masks, texture=&/T_SwoopClasp_pack}
  n2 = Constant2Vector() {g=2, r=2}
  n3 = DynamicParameter() {defaultValue=<rgba 3 0 1 1>, paramNames [IconIndex, "AnimBranch(0-1)", Param3, Param4]}
  n4 = StaticSwitchParameter(True=@n3.IconIndex, False=ScalarParameter() {parameterName=Index}) {parameterName=IsIcon}
  n5 = Multiply(A=Time(), B=ScalarParameter() {defaultValue=2.5, group=Background, parameterName=HologramSpeed})
  n6 = Multiply(
   A=ComponentMask(None=MaterialFunctionCall!(Texture=@n1, UVs, SubImages=@n2, Frame=@n4) {materialFunction=/Game/Global/Materials/Character_Parent_2023/VideoFlipbook_Parent_2025/MaterialFunctions/MF_FlipbookSubUV}.RGB) {b=true}
   B=LinearInterpolate(
    A=MaterialFunctionCall!(Time=@n5, MeshScaleIn, UVForScanline, CustomFlicker, UseCustomFlicker, UVForStroke, UVForGlitch, BaseUVIn, EmissiveIn, OpacityIn, WPOIn) {materialFunction=/Game/Tools/3DHB/Shader/Base/MF_3DHB_HologramFX}.UseForFX_GlitchToOpacityOut
    Alpha=@n3."AnimBranch(0-1)"
   ) {constA=1}
  )
  n7 = Step(Y=@n6, X=@n3."AnimBranch(0-1)") {constX=0.5}
  Multiply(
   A=StaticSwitchParameter(True=OneMinus(None=@n7), False=@n6) {parameterName=IsBranch}
   B=MaterialFunctionCall("FadeSensitivity (optional)"=Constant() {r=2}) {materialFunction=&FadeBasedOnViewAngle}.Result
  )
 ]
}
