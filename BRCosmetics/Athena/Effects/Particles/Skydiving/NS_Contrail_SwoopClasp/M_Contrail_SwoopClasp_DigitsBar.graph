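# Unlit translucent Niagara sprite material for the SwoopClasp skydiving contrail, drawing a bar of animated glyph digits tinted by particle color.
# Hashed UV cells, time and the particle Random parameter drive stepped values that pick glyphs from a tiled texture strip.
# A stepped gradient along U, blended by Size_Index, weights those values into a mask with thresholded glyphs, faded by view angle.
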Material /BRCosmetics/Athena/Effects/Particles/Skydiving/NS_Contrail_SwoopClasp/M_Contrail_SwoopClasp_DigitsBar
graph {
 properties {
  bAutomaticallySetUsageInEditor=false
  bUsedWithNiagaraSprites=true
  blendMode=BLEND_Translucent
  shadingModel=MSM_Unlit
  twoSided=true
  usedShadingModels=Unlit
 }
 nodes [
  MaterialFunctionCall(Emissive=ParticleColor().RGB) {materialFunction=&MF_FX_EmissiveAdjustments}
  n1 = Ceil(None=Multiply(A=TextureCoordinate(), B=Constant() {r=8}))
  n2 = Time()
  n3 = DynamicParameter() {defaultValue=<rgba 0 0 1 1>, paramNames [Size_Index,Random,Param3,Param4]}
  n4 = Frac(None=Add(A=MaterialFunctionCall(Vec2=@n1) {materialFunction=&MF_Hash}.Vec2, B=Add(A=@n2, B=@n3.Random)))
  n5 = OneMinus(None=ComponentMask(None=TextureCoordinate()) {r=true})
  n6 = ScalarParameter() {defaultValue=8, parameterName=Grid}
  n7 = Divide(A=Floor(None=Multiply(A=@n5, B=@n6)), B=@n6)
  n8 = Add(A=Saturate(None=@n4), B=Saturate(None=@n7))
  n9 = ComponentMask(None=Ceil(None=Multiply(A=@n8) {constB=7})) {r=true}
  n10 = Multiply(A=@n9, B=LinearInterpolate(A=Power(Base=@n7) {constExponent=5}, Alpha=@n3.Size_Index) {constA=5})
  n11 = TextureCoordinate()
  n12 = ComponentMask(None=@n11) {r=true}
  n13 = Constant() {r=8}
  n14 = Add(A=@n9, B=Fmod(A=Multiply(A=@n12, B=@n13) {constB=8}, B=Constant() {r=1})) {constB=7}
  n15 = LinearInterpolate(Alpha=@n3.Random) {constA=16, constAlpha=0, constB=32}
  n16 = ComponentMask(None=@n11) {g=true}
  n17 = TextureSample(UVs=AppendVector(A=Divide(A=@n14, B=@n15) {constB=16}, B=Multiply(A=@n16, B=@n13))) {
   samplerSource=SSM_Clamp_WorldGroupSettings
   samplerType=SAMPLERTYPE_Masks
   texture=&/T_Glyph_SwoopClasp
  }
  n18 = OneMinus(None=Step(Y=@n17.G) {constX=0.4})
  Multiply(
   A=Multiply(A=Saturate(None=@n10), B=Multiply(A=@n18))
   B=MaterialFunctionCall("FadeSensitivity (optional)"=Constant() {r=2}) {materialFunction=&FadeBasedOnViewAngle}.Result
  )
  Sine(None=@n4)
 ]
}
