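# Translucent, unlit, two-sided Niagara sprite material that draws an outlined progress bar for the SwoopClasp skydiving contrail.
# Particle color drives emissive; subtracting two scaled, rotatable UV masks forms the outline around a gradient fill.
# Dynamic parameters supply the timer, fill limit and minimum, with a debug blend to game time.
# A glitch-grid pixel texture thresholded by SpawnAnim animates the bar's appearance, and view-angle fading attenuates the result.
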Material /BRCosmetics/Athena/Effects/Particles/Skydiving/NS_Contrail_SwoopClasp/M_Contrail_SwoopClasp_ProgressBar
graph {
 properties {
  bAutomaticallySetUsageInEditor=false
  bUsedWithNiagaraSprites=true
  blendMode=BLEND_Translucent
  shadingModel=MSM_Unlit
  twoSided=true
  usedShadingModels=Unlit
 }
 nodes [
  MaterialFunctionCall() {materialFunction=&DebugTimeSine}
  MaterialFunctionCall(Emissive=ParticleColor().RGB) {materialFunction=&MF_FX_EmissiveAdjustments}
  n1 = NamedRerouteUsage()
  n2 = MaterialFunctionCall() {materialFunction=&LinearGradient}
  n3 = NamedRerouteUsage()
  n4 = DynamicParameter() {defaultValue=<rgba 1 0.5 0 1>, paramNames ["SpawnAnim(0-1)", "ProgressBarLimit(0-1)", "ProgressBarMin(0-1)", Param4]}
  n5 = Step(
   Y=Multiply(A=OneMinus(None=@n1), B=@n2.UGradient)
   X=Clamp(None=Multiply(A=@n3, B=@n4."ProgressBarLimit(0-1)") {constB=0.5}, Min=@n4."ProgressBarMin(0-1)")
  )
  n6 = Multiply(A=Multiply(A=@n5, B=@n2.UGradient) {constB=0.5}) {constB=0.5}
  n7 = TextureSample(UVs=MaterialFunctionCall!(UV=TextureCoordinate() {vTiling=0.25}) {materialFunction=/BRCosmetics/Weapons/FORT_Melee/Pickaxe_VoltaicHuntress/FX/Circuit/Functions/MF_Holog_Glitch_Grid}.Result) {texture=/BRCosmetics/Characters/Player/Male/Medium/Bodies/M_MED_PawJasmineGravity/FX/T_FX_Character_PawJasmineGravityGreet_Glitch_Pixel_Alpha}
  n8 = Step(Y=@n7.G, X=@n4."SpawnAnim(0-1)")
  Multiply(
   A=Multiply(A=Add(A=@n6, B=@n1), B=LinearInterpolate(A=@n8, B=@n5, Alpha=@n4."SpawnAnim(0-1)"))
   B=MaterialFunctionCall("FadeSensitivity (optional)"=Constant() {r=2}) {materialFunction=&FadeBasedOnViewAngle}.Result
  )
  n9 = MaterialFunctionCall(
   UVs=MaterialFunctionCall(UVs=TextureCoordinate(), Texture Scale=Constant() {r=1}) {materialFunction=&ScaleUVsByCenter}.UVs
   "Rotation Angle (0-1)"=ScalarParameter() {parameterName=Rotation}
  ) {materialFunction=&CustomRotator}
  n10 = ScalarParameter() {defaultValue=1, parameterName=Scale}
  n11 = Subtract(A=@n10, B=Multiply(A=Constant2Vector() {g=0.4, r=0.1}) {constB=0.75})
  NamedRerouteDeclaration(None=Subtract(
    A=MaterialFunctionCall(UVs=@n9.Rotated Values, Texture Scale=@n10) {materialFunction=&ScaleUVsByCenter}.0-1 mask
    B=MaterialFunctionCall(UVs=@n9.Rotated Values, Texture Scale=@n11) {materialFunction=&ScaleUVsByCenter}.0-1 mask
   )) {name=OutlineMask, nodeColor=<rgba 0.8047 0 1 1>}
  NamedRerouteDeclaration(None=LinearInterpolate(
    A=DynamicParameter() {parameterIndex=1, defaultValue=<rgba 1 0 0 0.5>, paramNames [Timer,Param2,Param3,Param4]}.Timer
    B=Time()
    Alpha=ScalarParameter() {parameterName=TimerDebug}
   )) {name=Time, nodeColor=<rgba 0.4062 1 0 1>}
  ScalarParameter() {defaultValue=0.15, parameterName=OutlineWidth, sliderMax=1}
 ]
}
