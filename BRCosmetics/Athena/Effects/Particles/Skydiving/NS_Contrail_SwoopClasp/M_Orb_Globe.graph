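# Unlit material for a glowing orb mesh particle in the SwoopClasp skydiving contrail cosmetic effect.
# A Fresnel term blends InnerColor and OuterColor parameters into a green rim-shaded gradient.
# That gradient is multiplied by a time-of-day tint from morning, evening and night colors, then by particle color.
# The result passes through the shared FX emissive adjustment function.
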Material /BRCosmetics/Athena/Effects/Particles/Skydiving/NS_Contrail_SwoopClasp/M_Orb_Globe
graph {
 properties {bAutomaticallySetUsageInEditor=false, bUsedWithNiagaraMeshParticles=true, shadingModel=MSM_Unlit, usedShadingModels=Unlit}
 nodes [
  n1 = LinearInterpolate(
   A=VectorParameter() {parameterName=InnerColor, defaultValue=<rgba 0 5 1.25 1>}.RGB
   B=VectorParameter() {parameterName=OuterColor, defaultValue=<rgba 0 0.8609 0.4304 1>}.RGB
   Alpha=Fresnel() {baseReflectFraction=0.005, exponent=0.5}
  )
  n2 = MaterialFunctionCall(
   Color_Morning=VectorParameter() {parameterName=Morning, defaultValue=<rgba 1 0.7478 0.7489 1>}.RGB
   Color_Evening=VectorParameter() {parameterName=Evening, defaultValue=<rgba 1 0.7478 0.7489 1>}.RGB
   Color_Night=VectorParameter() {parameterName=Night, defaultValue=<rgba 0.1842 0.228 0.35 1>}.RGB
  ) {materialFunction=&Color_Time_of_Day}
  MaterialFunctionCall(Emissive=Multiply(A=ParticleColor().RGB, B=Multiply(A=@n1, B=@n2.Result))) {materialFunction=&MF_FX_EmissiveAdjustments}
 ]
}
