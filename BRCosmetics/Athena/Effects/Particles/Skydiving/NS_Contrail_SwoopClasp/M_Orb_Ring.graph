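# Unlit, masked, two-sided material for Niagara mesh particles, drawing glowing orb rings in the SwoopClasp skydiving contrail.
# Emissive color blends outer and inner colors by a cubed packed-texture mask, then applies time-of-day tint, particle color and FX emissive adjustments.
# A rotation around the mesh particle pivot picks X or Y axis from vertex red, with direction from a dynamic parameter.
# A particle-driven timer supplies the angle, blendable to game time for debugging.
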Material /BRCosmetics/Athena/Effects/Particles/Skydiving/NS_Contrail_SwoopClasp/M_Orb_Ring
graph {
 properties {
  bAutomaticallySetUsageInEditor=false
  bUsedWithNiagaraMeshParticles=true
  blendMode=BLEND_Masked
  shadingModel=MSM_Unlit
  twoSided=true
  usedShadingModels=Unlit
 }
 nodes [
  n1 = MaterialFunctionCall(
   Color_Morning=VectorParameter() {parameterName=Morning, defaultValue=<rgba 1 0.7478 0.7489 1>}.RGB
   Color_Evening=VectorParameter() {parameterName=Evening, defaultValue=<rgba 1 0.7478 0.7489 1>}.RGB
   Color_Night=VectorParameter() {parameterName=Night, defaultValue=<rgba 0.1842 0.228 0.35 1>}.RGB
  ) {materialFunction=&Color_Time_of_Day}
  n2 = TextureSample(UVs=TextureCoordinate()) {samplerType=SAMPLERTYPE_Masks, texture=&/T_SwoopClasp_pack}
  n3 = LinearInterpolate(
   A=VectorParameter() {parameterName=OuterColor, defaultValue=<rgba 0 0.8609 0.4304 1>}.RGB
   B=VectorParameter() {parameterName=InnerColor, defaultValue=<rgba 0 5 1.25 1>}.RGB
   Alpha=Saturate(None=Power(Base=@n2.R) {constExponent=3})
  )
  MaterialFunctionCall(Emissive=Multiply(A=ParticleColor().RGB, B=Multiply(A=@n1.Result, B=@n3))) {materialFunction=&MF_FX_EmissiveAdjustments}
  NamedRerouteDeclaration(None=LinearInterpolate(
    A=DynamicParameter() {defaultValue=<rgba 1 1 0 0.5>, paramNames [Timer, "AnimCurveSpawn(0-1)", "BoolRandom(0-1)", Param4]}.Timer
    B=Time()
    Alpha=ScalarParameter() {parameterName=TimerDebug}
   )) {name=Time, nodeColor=<rgba 0.4062 1 0 1>}
  n4 = VertexColor()
  n5 = Constant3Vector() {constant=<rgba 1 0 0 0>}
  n6 = Constant3Vector() {constant=<rgba 0 1 0 0>}
  RotateAboutAxis(
   NormalizedRotationAxis=Multiply(
    A=If(A=@n4.R, "A > B"=@n5, "A == B"=@n6, "A < B"=@n5) {constB=1}
    B=DynamicParameter() {paramNames [Param1, Param2, RotationDIrection, Param4]}.RotationDIrection
   )
   RotationAngle=NamedRerouteUsage()
   PivotPoint=MaterialFunctionCall() {materialFunction=&ObjectPivotPoint}.Mesh Particle Pivot Location
   Position=WorldPosition().XYZ
  )
  Time()
 ]
}
