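# Unlit, two-sided translucent material for the orb ring mesh particles in the SwoopClasp skydiving contrail effect.
# A center-peaked U gradient blends outer to inner color, tinted by time of day for emissive, and is multiplied by a view-angle fade.
# Geometry rotates around the particle pivot, with vertex red choosing X or Y axis and a dynamic parameter setting direction.
# Rotation time comes from a particle Timer parameter, blendable to global Time for debugging.
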Material /BRCosmetics/Athena/Effects/Particles/Skydiving/NS_Contrail_SwoopClasp/M_Orb_Ring_SplineThicken
graph {
 properties {
  bAutomaticallySetUsageInEditor=false
  bUsedWithNiagaraMeshParticles=true
  blendMode=BLEND_Translucent
  shadingModel=MSM_Unlit
  twoSided=true
  usedShadingModels=Unlit
 }
 nodes [
  n1 = VectorParameter() {parameterName=Morning, defaultValue=<rgba 1 0.7478 0.7489 1>}
  n2 = VectorParameter() {parameterName=Evening, defaultValue=<rgba 1 0.7478 0.7489 1>}
  n3 = VectorParameter() {parameterName=Night, defaultValue=<rgba 0.2699 0.3376 0.5 1>}
  n4 = MaterialFunctionCall() {materialFunction=&LinearGradient}
  n5 = Multiply(A=@n4.UGradient, B=OneMinus(None=@n4.UGradient))
  n6 = LinearInterpolate(
   A=VectorParameter() {parameterName=OuterColor, defaultValue=<rgba 0 0.8609 0.4448 1>}.RGB
   B=VectorParameter() {parameterName=InnerColor, defaultValue=<rgba 0 10 5 1>}.RGB
   Alpha=@n5
  )
  MaterialFunctionCall(Emissive=Multiply(
    A=MaterialFunctionCall(Color_Morning=@n1.RGB, Color_Evening=@n2.RGB, Color_Night=@n3.RGB) {materialFunction=&Color_Time_of_Day}.Result
    B=Multiply(A=@n6) {constB=0.5}
   )) {materialFunction=&MF_FX_EmissiveAdjustments}
  Multiply(
   A=@n5
   B=MaterialFunctionCall("FadeSensitivity (optional)"=Constant() {r=2}) {materialFunction=&FadeBasedOnViewAngle}.Result
  )
  NamedRerouteDeclaration(None=LinearInterpolate(
    A=DynamicParameter() {defaultValue=<rgba 1 1 0 0.5>, paramNames [Timer, "AnimCurveSpawn(0-1)", "BoolRandom(0-1)", Param4]}.Timer
    B=Time()
    Alpha=ScalarParameter() {parameterName=TimerDebug}
   )) {name=Time, nodeColor=<rgba 0.4062 1 0 1>}
  n7 = Constant3Vector() {constant=<rgba 1 0 0 0>}
  n8 = If(A=VertexColor().R, "A > B"=@n7, "A == B"=Constant3Vector() {constant=<rgba 0 1 0 0>}, "A < B"=@n7) {constB=1}
  n9 = DynamicParameter() {paramNames [Param1, Param2, RotationDIrection, Param4]}
  Reroute(None=RotateAboutAxis(
    NormalizedRotationAxis=Multiply(A=@n8, B=@n9.RotationDIrection)
    RotationAngle=NamedRerouteUsage()
    PivotPoint=MaterialFunctionCall() {materialFunction=&ObjectPivotPoint}.Mesh Particle Pivot Location
    Position=WorldPosition().XYZ
   ))
 ]
}
