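# Unlit translucent Niagara material for the SwoopClasp skydiving contrail's DNA ribbon and equal-symbol meshes.
# World position offset spins meshes around their particle pivot and adds a height-weighted sine wave, with corrected normals.
# Vertex green selects DNA or symbol texture channels for shape masks, broken up by noise-blended hologram glitch, with two-tone emissive color.
# Opacity combines vertex color, fresnel, particle alpha and a height fade shifted by the Dive dynamic parameter.
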Material /BRCosmetics/Athena/Effects/Particles/Skydiving/NS_Contrail_SwoopClasp/M_SwoopClasp_DNA_WPO
graph {
 properties {
  bUsedWithNiagaraMeshParticles=true
  bUsedWithNiagaraSprites=true
  blendMode=BLEND_Translucent
  pixelDepthOffsetMode=PDOM_Legacy
  previewMesh=&/SM_SwoopClasp_DNA_Ribbon
  shadingModel=MSM_Unlit
  twoSided=true
  usedShadingModels=Unlit
 }
 nodes [
  n1 = LocalPosition()
  ComponentMask(None=@n1.XYZ) {b=true}
  n2 = Normalize(VectorInput=Transform(None=Constant3Vector() {constant=<rgba 0 0 1 0>}) {transformSourceType=TRANSFORMSOURCE_Instance})
  n3 = NamedRerouteUsage()
  n4 = DynamicParameter() {defaultValue=<rgba 0.4 1 1 1>, paramNames [RandOffset,Time,EqualOffset,Dive]}
  n5 = Fmod(A=Multiply(A=Add(A=@n3, B=@n4.RandOffset), B=ScalarParameter() {defaultValue=-0.6, parameterName=SpinSPeed}), B=Constant() {r=1})
  NamedRerouteDeclaration(None=MaterialFunctionCall(Rotation Axis=@n2, Rotation Angle=@n5) {materialFunction=&FixRotateAboutAxisNormals}.New World Space Vertex Normal) {name=FixVertexNormalWPO, nodeColor=<rgba 1 0.2812 0 1>}
  n6 = NamedRerouteUsage()
  n7 = ComponentMask(None=Subtract(A=ActorPositionWS(), B=WorldPosition().XYZ)) {b=true}
  n8 = ScalarParameter() {defaultValue=256, parameterName=WPOGradientSize}
  n9 = Saturate(None=Divide(A=Add(A=@n7, B=@n8), B=@n8))
  n10 = Subtract(A=WorldPosition().XYZ, B=Multiply(A=ActorPositionWS()))
  n11 = NamedRerouteUsage()
  n12 = Multiply(A=DynamicParameter() {defaultValue=<rgba 0.4 1 1 1>, paramNames [RandOffset,Time,EqualOffset,Dive]}.RandOffset) {constB=2}
  n13 = Add(
   A=Divide(A=ComponentMask(None=@n10) {b=true}, B=ScalarParameter() {defaultValue=512, parameterName=WPOWaveLength})
   B=Multiply(A=Add(A=@n11, B=@n12), B=ScalarParameter() {defaultValue=0.6, parameterName=WPOWavePanSpeed})
  )
  n14 = AppendVector(A=Sine(None=@n13), B=Sine(None=@n13) {period=0.73})
  n15 = Constant()
  n16 = Multiply(
   A=Power(Base=OneMinus(None=@n9)) {constExponent=1}
   B=Multiply(A=AppendVector(A=@n14, B=@n15), B=ScalarParameter() {defaultValue=10, parameterName=WPOWaveMovement})
  )
  NamedRerouteDeclaration(None=Add(A=@n6, B=Multiply(A=@n16))) {name=WPO, nodeColor=<rgba 0 1 1 1>}
  n17 = VectorParameter() {parameterName=TotalUVControl, channelNames {a=V_Speed, b=U_Speed, g=V_Scale, r=U_Scale}, defaultValue=<rgba 1 1 0 0>}
  NamedRerouteDeclaration(None=MaterialFunctionCall!(UVInput=TextureCoordinate(), TimeInput=NamedRerouteUsage(), Control_RGB=@n17, Control_A=@n17) {materialFunction=/Game/Effects/Niagara/MasterMaterials/Functions/Common/FXCommon_QuickUVControl}.UV) {name=UV, nodeColor=<rgba 0 1 0.3672 1>}
  NamedRerouteDeclaration(None=RotateAboutAxis(
    NormalizedRotationAxis=@n2
    RotationAngle=@n5
    PivotPoint=MaterialFunctionCall() {materialFunction=&ObjectPivotPoint}.Mesh Particle Pivot Location
    Position=WorldPosition().XYZ
   )) {name=RotateAboutAxis_WPO, nodeColor=<rgba 1 0.6797 0 1>}
  n18 = Add(
   A=@n1.Z
   B=LinearInterpolate(Alpha=DynamicParameter() {defaultValue=<rgba 0.4 1 1 1>, paramNames [RandOffset,Time,EqualOffset,Dive]}.Dive) {constA=150, constB=0}
  ) {constB=0}
  n19 = Multiply(A=OneMinus(None=Divide(A=@n18) {constB=350})) {constB=0.5}
  n20 = VertexColor()
  n21 = Power(Base=@n20.R) {constExponent=2.5}
  n22 = Multiply(A=Multiply(A=Saturate(None=@n19), B=StaticSwitchParameter(True=@n20.G, False=@n21) {parameterName=EqualSymbol}), B=ParticleColor().A)
  n23 = MaterialFunctionCall(
   Brightness=ScalarParameter() {defaultValue=1, parameterName=FresnelMult}
   Exponent=ScalarParameter() {defaultValue=1, parameterName=FresnelExp}
   Vertex / Pixel Normal=NamedRerouteUsage()
  ) {materialFunction=&MF_Fresnel_FX}
  NamedRerouteDeclaration(None=Multiply(A=NamedRerouteUsage(), B=Multiply(A=@n22, B=@n23.Fresnel Double Sided))) {name=Opacity, nodeColor=<rgba 0.3594 0 1 1>}
  NamedRerouteDeclaration(None=LinearInterpolate(
    A=DynamicParameter() {defaultValue=<rgba 0.4 1 1 1>, paramNames [RandOffset,Time,EqualOffset,Dive]}.Time
    B=Time()
    Alpha=ScalarParameter() {parameterName=TimerDebug}
   )) {name=Time, nodeColor=<rgba 0 1 0.2734 1>}
  n24 = VectorParameter() {parameterName=OutlineColor, defaultValue=<rgba 0 0.75 0.3125 1>}
  n25 = Multiply(A=ParticleColor().RGB, B=VectorParameter() {parameterName=InnerColor, defaultValue=<rgba 1 1 1 0>}.RGB)
  n26 = Saturate(None=NamedRerouteUsage())
  NamedRerouteDeclaration(None=MaterialFunctionCall(Emissive=LinearInterpolate(A=@n24.RGB, B=@n25, Alpha=@n26)) {materialFunction=&MF_FX_EmissiveAdjustments}.Result) {name=Emissive, nodeColor=<rgba 1 0 0.2109 1>}
  n27 = TextureObjectParameter() {
   group=DNA Shape
   parameterName=Symbol_Shape_Texture
   sortPriority=0
   texture=&/T_SwoopClasp_Symbol
  }
  n28 = Constant()
  n29 = DynamicParameter() {defaultValue=<rgba 0.4 1 1 1>, paramNames [RandOffset,Time,EqualOffset,Dive]}
  n30 = VectorParameter() {group=NoiseB, parameterName=Equal_Shape_UVControl, channelNames {a=VPan, b=UPan, g=VScale, r=UScale}, defaultValue=<rgba 1 4 0 0.6>}
  n31 = NamedRerouteUsage()
  n32 = Constant2Vector() {g=0.5, r=0.5}
  n33 = Power(Base=MaterialFunctionCall(
    Texture=@n27
    AdditionalUVOffset=AppendVector(A=@n28, B=@n29.EqualOffset)
    UVControl_RGB=@n30
    UVControl_A=@n30
    "[Override]TimeInput"=NamedRerouteUsage()
    "[Override]UVInput"=MaterialFunctionCall(UVs=@n31, Texture Scale=@n32) {materialFunction=&ScaleUVsByCenter}.UVs
   ) {materialFunction=&FXCommon_TextureSample}.G) {constExponent=5}
  n34 = TextureObjectParameter() {
   group=DNA Shape
   parameterName=DNA_Shape_Texture
   sortPriority=0
   texture=&T_SwoopClasp_Symbol
  }
  n35 = VectorParameter() {group=NoiseB, parameterName=DNA_Shape_UVControl, channelNames {a=VPan, b=UPan, g=VScale, r=UScale}, defaultValue=<rgba 1 1 0 0>}
  n36 = MaterialFunctionCall(
   Texture=@n34
   UVControl_RGB=@n35
   UVControl_A=@n35
   "[Override]UVInput"=MaterialFunctionCall(XY=NamedRerouteUsage()) {materialFunction=&Swizzle}.YX
  ) {materialFunction=&FXCommon_TextureSample}
  NamedRerouteDeclaration(None=If(A=VertexColor().G, "A > B"=@n33, "A == B"=@n33, "A < B"=@n36.A) {constB=1}) {name=OpacityMask, nodeColor=<rgba 0 0.5312 1 1>}
  n37 = WorldPosition()
  n38 = ComponentMask(None=@n37.XY) {r=true}
  n39 = Divide(A=@n37.Z) {constB=512}
  n40 = MaterialFunctionCall!(
   Time=Multiply(A=Time(), B=ScalarParameter() {defaultValue=2.5, group=Background, parameterName=HologramSpeed})
   MeshScaleIn
   UVForScanline=Multiply(A=AppendVector(A=@n38, B=@n39), B=Constant2Vector() {g=2.5, r=1})
   CustomFlicker
   UseCustomFlicker
   UVForStroke
   UVForGlitch
   BaseUVIn
   EmissiveIn
   OpacityIn
   WPOIn
  ) {materialFunction=/Game/Tools/3DHB/Shader/Base/MF_3DHB_HologramFX}
  n41 = Multiply(A=@n40.UseForFX_GlitchToOpacityOut, B=@n40.UseForFX_GlitchToOpacityOut)
  n42 = Multiply(A=@n36.B) {constB=2.5}
  n43 = If(A=VertexColor().G, "A > B"=@n42, "A == B"=@n42, "A < B"=@n36.R) {constB=1}
  n44 = Panner(Coordinate=TextureCoordinate() {uTiling=0.1, vTiling=2}) {speedX=-0.5}
  NamedRerouteDeclaration(None=LinearInterpolate(
    A=@n43
    B=Multiply(A=@n43, B=@n41)
    Alpha=TextureSample(UVs=@n44) {samplerType=SAMPLERTYPE_LinearColor, texture=&T_NoisePack_LinearColor}.RGB
   )) {name=ShapeMaskOpacity, nodeColor=<rgba 1 0.0234 0 1>}
  NamedRerouteUsage()
  NamedRerouteUsage()
  NamedRerouteUsage()
  StaticSwitchParameter(True=Constant(), False=@n6) {parameterName=EqualSymbol}
 ]
}
