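# Skydiving contrail cosmetic (SwoopClasp) made of glowing green socket-attached spheres, DNA meshes and HUD-style sprites.
# Spheres (mesh, sprite, light) and Mesh_DNA each place two particles through a socket reader, with DNA meshes offset by owner velocity.
# HUD merges four burst groups into one emitter, using spawn-group and partition switches for layout.
# System dive alphas blend offsets, mesh tilt and material animation, with front-end overrides for preview.
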NiagaraSystem /BRCosmetics/Athena/Effects/Particles/Skydiving/NS_Contrail_SwoopClasp/NS_Contrail_SwoopClasp
dependencies {
 usedDataInterfaces [
  &NiagaraDataInterfaceCurve
  &NiagaraDataInterfaceColorCurve
  &NiagaraDataInterfacePlatformSet
  &NiagaraDataInterfaceSocketReader
 ]
 usedDynamicInputs [
  &MakeIntFromFloat
  &FN_PickFloatDependOnPlatform
  /Niagara/DynamicInputs/Lerp/SmoothLerpOverTime_Float
 ]
 usedModules [
  &self:SystemSpawnScript.SystemScriptSource.SystemScriptGraph.NiagaraNodeAssignment_1.SetVariables_#1
  &self:SystemSpawnScript.SystemScriptSource.SystemScriptGraph.NiagaraNodeAssignment_3.SetVariables_#2
  &self:SystemSpawnScript.SystemScriptSource.SystemScriptGraph.NiagaraNodeAssignment_0.SetVariables_#3
  &self:SystemSpawnScript.SystemScriptSource.SystemScriptGraph.NiagaraNodeAssignment_6.SetVariables_#4
  &self:SystemSpawnScript.SystemScriptSource.SystemScriptGraph.NiagaraNodeAssignment_5.SetVariables_#5
  &SystemState
  /Game/Effects/Niagara/Modules/Skydiving/FN_DivingAttributes/v5/FN_DivingAttributes
  &self:SystemSpawnScript.SystemScriptSource.SystemScriptGraph.NiagaraNodeAssignment_2.SetVariables_#6
  &self:SystemSpawnScript.SystemScriptSource.SystemScriptGraph.NiagaraNodeAssignment_8.SetVariables_#7
 ]
}
emitters {
 HUD {
  emitterSpawnScript [/SetVariables() {isSetParameters=true}]
  emitterUpdateScript [
   &EmitterState(Life Cycle Mode=::NewEnumerator1, Loop Behavior=::NewEnumerator1, Loop Duration Mode=::NewEnumerator1, UseLoopDelay=false)
   &SpawnBurst_Instantaneous(Spawn Count=2, Spawn Time=0.3, Use Spawn Probability=false, UseLoopCountLimit=false)
   &SpawnBurst_Instantaneous(Spawn Count=4, Spawn Group=1, Spawn Time=1, Use Spawn Probability=false, UseLoopCountLimit=false) {module=SpawnBurst_Instantaneous001}
   &SpawnBurst_Instantaneous(Spawn Count=8, Spawn Group=2, Spawn Time=1, Use Spawn Probability=false, UseLoopCountLimit=false) {module=SpawnBurst_Instantaneous002}
   &SpawnBurst_Instantaneous(Spawn Count=6, Spawn Group=3, Spawn Time=0.4, Use Spawn Probability=false, UseLoopCountLimit=false) {module=SpawnBurst_Instantaneous003}
   /SetVariables(Emitter.CustomTime:NiagaraFloat=&Add_Float(A=$Emitter.CustomTime, B=&Lerp_Float(A=0.015, B=0.1, Alpha=$System.SmoothDiveAlpha))) {isSetParameters=true}
  ]
  particleSpawnScript [
   /SetVariables(Particles.VisibilityTag=$Emitter.SpawnGroup) {isSetParameters=true}
   &PartitionParticles(Number Of Partitions=&MergeHelper_IntSwitch(0=2, 1=4, 2=8, 3=6, 4=?, 5=?, Index=$Particles.VisibilityTag, Num=::NewEnumerator2))
   &InitializeParticle(
    Color=&MergeHelper_ColorSwitch(
     0=<rgba 0.5 10 2.655 1>
     1=<rgba 0.5 10 2.655 1>
     2=<rgba 0 5 1.1342 1>
     3=<rgba 0.5 10 2.655 1>
     Index=$Particles.VisibilityTag
     Num=::NewEnumerator2
    )
    Lifetime=2
    Sprite Rotation Angle=&MergeHelper_FloatSwitch(
     0=0
     1=0
     2=0
     3=&MergeHelper_FloatSwitch(0=0, 1=180, 2=180, 3=0, 4=0, 5=0, Index=$Particles.PartitionParticles.Partition, Num=::NewEnumerator4)
     Index=$Particles.VisibilityTag
     Num=::NewEnumerator2
    )
    Sprite Rotation Mode=::NewEnumerator0
    Sprite Size=&MergeHelper_Vector2Switch(
     0=<25 25>
     1=&MergeHelper_Vector2Switch(0=<75 75>, 1=<75 75>, 2=<40 40>, 3=<40 40>, 4=?, Index=$Particles.PartitionParticles.Partition, Num=::NewEnumerator2)
     2=<30 5>
     3=<25 25>
     4=?
     Index=$Particles.VisibilityTag
     Num=::NewEnumerator2
    )
    Sprite Size Mode=::NewEnumerator4
    Sprite UV Mode=::NewEnumerator4
    Sprite UV Scale:Vector2f=&MergeHelper_Vector2Switch(
     0=<1 1>
     1=<1 1>
     2=<1 1>
     3=&MergeHelper_Vector2Switch(0=<1 1>, 1=<1 1>, 2=<1 1>, 3=<1 -1>, 4=<1 -1>, 5=<-1 1>, Index=$Particles.PartitionParticles.Partition, Num=::NewEnumerator4)
     4=?
     Index=$Particles.VisibilityTag
     Num=::NewEnumerator2
    )
    UsePositionOffset=false
    Write Material Random=true
   )
   /SetVariables(Particles.DynamicParameter_Icon:Vector4f=/Game/Effects/Niagara/Modules/MergeEmitters/MergeHelper_Vector4Switch(
     0:Vector4f=<1 0 0 0>
     1:Vector4f=<2 0 0 0>
     2:Vector4f=?
     3:Vector4f=?
     4:Vector4f=?
     5:Vector4f=?
     6:Vector4f=?
     7:Vector4f=?
     8:Vector4f=?
     9:Vector4f=?
     Index:NiagaraInt32=$Particles.PartitionParticles.Partition
     Num:MergeHelper_SwitcherEnum=::NewEnumerator0
    )) {isSetParameters=true}
   &SpriteFacingAndAlignment(
    Alignment Coordinate Space=::Local
    Facing Coordinate Space=::Local
    Sprite Alignment=<0 0 1>
    Sprite Facing=<0 -1 0.5>
    WriteAlignment=true
    WriteFacing=true
   )
   &SystemLocation(
    MaskSpawnGroup=false
    Offset=&MergeHelper_VectorSwitch(
     0=&Add_Vector(
      A=&Multiply_Vector(
       A=$System.Offset_UI
       B=&MergeHelper_VectorSwitch(
        0=<1 1 1>
        1=<-1 1 1>
        2=<1 1 1>
        3=<-1 1 1>
        Index=&ModuloInt(A=$Particles.PartitionParticles.Partition, B=2)
        Num=::NewEnumerator2
       )
      )
      B=&MergeHelper_VectorSwitch(0=<-30 25 25>, 1=<20 25 25>, 2=?, 3=?, Index=$Particles.PartitionParticles.Partition, Num=::NewEnumerator0)
     )
     1=&Add_Vector(
      A=&Multiply_Vector(
       A=$System.Offset_UI
       B=&MergeHelper_VectorSwitch(
        0=<1 1 1>
        1=<-1 1 1>
        2=<1 1 1>
        3=<-1 1 1>
        Index=&ModuloInt(A=$Particles.PartitionParticles.Partition, B=2)
        Num=::NewEnumerator2
       )
      )
      B=&MergeHelper_VectorSwitch(0=<-5 35 10>, 1=<-35 30 10>, 2=<12.5 35 20>, 3=<-17.5 30 20>, Index=$Particles.PartitionParticles.Partition, Num=::NewEnumerator2)
     )
     2=&Add_Vector(
      A=&Multiply_Vector(
       A=$System.Offset_UI
       B=&MergeHelper_VectorSwitch(
        0=<1 1 1>
        1=<-1 1 1>
        2=<1 1 1>
        3=<-1 1 1>
        Index=&ModuloInt(A=$Particles.PartitionParticles.Partition, B=2)
        Num=::NewEnumerator2
       )
      )
      B=&MergeHelper_VectorSwitch(
       0=<0 30 25>
       1=<40 40 20>
       2=<0 25 17.5>
       3=<40 45 27.5>
       4=<-75 30 50>
       5=<-30 30 17.5>
       6=<-75 25 42.5>
       7=<-30 25 10>
       Index=$Particles.PartitionParticles.Partition
       Num=::NewEnumerator6
      )
     )
     3=&Add_Vector(
      A=&Multiply_Vector(
       A=$System.Offset_UI
       B=&MergeHelper_VectorSwitch(
        0=<1 1 1>
        1=<-1 1 1>
        2=<1 1 1>
        3=<-1 1 1>
        Index=&ModuloInt(A=$Particles.PartitionParticles.Partition, B=2)
        Num=::NewEnumerator2
       )
      )
      B=&MergeHelper_VectorSwitch(
       0=<-7.4374 25 22.5>
       1=<-5 35 30.3728>
       2=<-46.4817 25 45.8405>
       3=<40.7258 35 26.2585>
       4=<-11.1962 35 34.1051>
       5=<-0.9202 25 15.4913>
       Index=$Particles.PartitionParticles.Partition
       Num=::NewEnumerator4
      )
     )
     Index=$Particles.VisibilityTag
     Num=::NewEnumerator2
    )
   )
  ]
  particleUpdateScript [
   &ParticleState(Kill Particles When Lifetime Has Elapsed=false, Let Infinitely Lived Particles Die When Emitter Deactivates=true)
   &ScaleSpriteSize(
    Scale Sprite Size Mode=::NewEnumerator0
    Uniform Scale Factor=&MergeHelper_FloatSwitch(
     0=&FloatFromCurve(FloatCurve=DataInterface() {propertyValues {Curve {keys [
          <at 0 0 RCIM_Cubic>
          {arriveTangent=2.637, interpMode=RCIM_Cubic, leaveTangent=2.637, time=0.1724, value=0.8208}
          {arriveTangent=1.8766, interpMode=RCIM_Cubic, leaveTangent=1.8766, tangentMode=RCTM_User, time=0.2484, value=0.655}
          {arriveTangent=-0.1395, interpMode=RCIM_Cubic, leaveTangent=-0.1395, tangentMode=RCTM_User, time=0.3526, value=1.0706}
          {arriveTangent=-0.109, interpMode=RCIM_Cubic, leaveTangent=-0.109, time=0.3962, value=0.9794}
          <at 1 1 RCIM_Cubic>
         ]}}})
     1=1
     2=1
     3=1
     Index=$Particles.VisibilityTag
     Num=::NewEnumerator2
    )
   )
   &SpriteRotationRate(Rotation Rate=&MergeHelper_FloatSwitch(0=25, 1=0, 2=0, 3=0, 4=0, Index=$Particles.VisibilityTag, Num=::NewEnumerator3))
   &ScaleColor(
    Scale Alpha=&MergeHelper_FloatSwitch(
     0=&FloatFromCurve(FloatCurve=DataInterface() {propertyValues {Curve {keys [<at -0.0046 1.0113 RCIM_Cubic>,<at 1.0011 0.002 RCIM_Cubic>]}}})
     1=&FloatFromCurve(FloatCurve=DataInterface() {propertyValues {Curve {keys [<at -0.0046 1.0113 RCIM_Cubic>,<at 1.0011 0.002 RCIM_Cubic>]}}})
     2=1
     3=&FloatFromCurve(FloatCurve=DataInterface() {propertyValues {Curve {keys [
          {arriveTangent=-6.0661, interpMode=RCIM_Cubic, leaveTangent=-6.0661, tangentMode=RCTM_User, time=-0.0046, value=1.0113}
          <at 0.5 0.002 RCIM_Cubic>
         ]}}})
     Index=$Particles.VisibilityTag
     Num=::NewEnumerator2
    )
    ScaleA=true
    ScaleRGB=false
   )
   &DynamicMaterialParameters(
    Index 0 Param 1=&MergeHelper_FloatSwitch(
     0=1
     1=&MergeHelper_FloatSwitch(0=0, 1=0, 2=1, 3=1, Index=$Particles.PartitionParticles.Partition, Num=::NewEnumerator2)
     2=&FloatFromCurve(
      FloatCurve=DataInterface() {propertyValues {Curve {
         defaultValue=1
         keys [
          {interpMode=RCIM_Cubic, tangentMode=RCTM_User, time=0, value=0}
          {arriveTangent=4.7019, interpMode=RCIM_Cubic, leaveTangent=4.7019, tangentMode=RCTM_User, time=0.25, value=1}
         ]
        }}}
      CurveIndex=&Multiply_Float(A=$Particles.NormalizedAge, B=&RandomRangeFloat(Minimum=0.25, Maximum=2))
     )
     3=3
     Index=$Particles.VisibilityTag
     Num=::NewEnumerator2
    )
    Index 0 Param 2=&MergeHelper_FloatSwitch(
     0=0
     1=&MergeHelper_FloatSwitch(0=0, 1=1, 2=?, 3=?, Index=$Particles.PartitionParticles.Partition, Num=::NewEnumerator0)
     2=&Lerp_Float(A=&RandomRangeFloat(Minimum=0.33, Maximum=0.75), B=1, Alpha=$System.SmoothDiveAlpha)
     3=&FloatFromCurve(FloatCurve=DataInterface() {propertyValues {Curve {
         defaultValue=1
         keys [
          {arriveTangent=-5.5005, interpMode=RCIM_Cubic, leaveTangent=-5.5005, tangentMode=RCTM_User, time=0, value=1}
          {interpMode=RCIM_Cubic, tangentMode=RCTM_User, time=0.5, value=0.1}
         ]
        }}})
     Index=$Particles.VisibilityTag
     Num=::NewEnumerator2
    )
    Index 0 Param 3=&MergeHelper_FloatSwitch(
     0=0
     1=0
     2=&Lerp_Float(A=0, B=0.8, Alpha=$System.SmoothDiveAlpha)
     3=0
     Index=$Particles.VisibilityTag
     Num=::NewEnumerator2
    )
    Index 1 Param 1=&Waveform(
     Global Input Position=&RandomRangeFloat(Minimum=0, Maximum=1)
     Global Phase=$Emitter.CustomTime
     "Waveform [1]"=::NewEnumerator10
     "[1] Frequency"=2
     "[1] Amplitude Min/Max"=&Lerp_Vector2(A=<0 1>, B=<0.75 1>, Alpha=$System.SmoothDiveAlpha)
    )
    Module.Index0Float0Write=true
    Module.Index0Float1Write=true
    Module.Index0Float2Write=true
    Module.Index0Float3Write=true
    Module.Index1Float0Write=true
    Module.Index1Float1Write=true
    Module.Index1Float2Write=true
    Module.Index1Float3Write=true
    Write Parameter Index 1=::NewEnumerator1
   )
  ]
  rendererClasses [&NiagaraSpriteRendererProperties]
  renderers [
   {rendererClass=&NiagaraSpriteRendererProperties}
   {rendererClass=&NiagaraSpriteRendererProperties, rendererIndex=1}
   {rendererClass=&NiagaraSpriteRendererProperties, rendererIndex=2}
   {rendererClass=&NiagaraSpriteRendererProperties, rendererIndex=3}
   {rendererClass=&NiagaraSpriteRendererProperties, rendererIndex=4}
  ]
 }
 Mesh_DNA {
  emitterSpawnScript [/SetVariables() {isSetParameters=true}]
  emitterUpdateScript [
   &EmitterState(
    Inactive Response=::NewEnumerator1
    Life Cycle Mode=::NewEnumerator1
    Loop Behavior=::NewEnumerator1
    Loop Duration Mode=::NewEnumerator1
    UseLoopDelay=false
   )
   &SpawnBurst_Instantaneous(Spawn Count=2, Use Spawn Probability=false, UseLoopCountLimit=false)
   /SetVariables(Emitter.CustomTime:NiagaraFloat=&Add_Float(A=$Emitter.CustomTime, B=&Lerp_Float(A=0.025, B=0.075, Alpha=$System.SmoothDiveAlpha))) {isSetParameters=true}
  ]
  particleSpawnScript [
   &PartitionParticles(Write Partition to Ribbon ID=::NewEnumerator0)
   &InitializeParticle(
    Color=<rgba 0.5 10 2.655 1>
    Lifetime=5
    Mesh Scale=<0.6 0.6 0.45>
    Mesh Scale Mode=::NewEnumerator4
    UsePositionOffset=false
    Write Material Random=true
   )
   /SetVariables() {isSetParameters=true}
   &InitialMeshOrientation(
    Mesh Orientation Mode=::NewEnumerator2
    Rotation=&MergeHelper_VectorSwitch(0=<-0.025 0.025 0>, 1=<-0.025 -0.025 0>, 2=?, 3=?, Index=$Particles.PartitionParticles.Partition, Num=::NewEnumerator0)
    Use Rotation Vector=true
   )
  ]
  particleUpdateScript [
   &ParticleState(Kill Particles When Lifetime Has Elapsed=false)
   &SocketLocation(
    Orientation Sampling=::NewEnumerator1
    SamplePosOffset=true
    Sampled Position Offset=&Lerp_Vector(
     A=<0 15 -10>
     B=&MergeHelper_VectorSwitch(0=<15 10 25>, 1=<-15 10 25>, 2=?, 3=?, Index=$Particles.PartitionParticles.Partition, Num=::NewEnumerator0)
     Alpha=$System.ContrailDiveAlpha
    )
    Sampled Position Offset Coordinate Space:ENiagaraSkelMeshTransforms=::NewEnumerator1
    Socket Index=$Particles.PartitionParticles.Partition
    Socket Reader=$System.Socket Reader
    Socket Sampling Mode=::NewEnumerator3
    VelocitySpeedLimit=false
   )
   /SetVariables(Particles.MeshOrientation=&MultiplyQuaternion(
     Quaternion A=$Particles.Initial.MeshOrientation
     Quaternion B=&MakeQuaternion(
      Quaternion From=0
      Axis=&FN_FrontEndVector3Modifier(
       Gameplay=&MergeHelper_VectorSwitch(0=<-1 1 0>, 1=<-1 -1 0>, 2=?, 3=?, Index=$Particles.PartitionParticles.Partition, Num=::NewEnumerator0)
       Front End=&MergeHelper_VectorSwitch(0=<1 0 0>, 1=<1 0 0>, 2=?, 3=?, Index=$Particles.PartitionParticles.Partition, Num=::NewEnumerator0)
      )
      Angle=&Lerp_Float(A=0, B=&FN_FrontEndFloatModifier(Gameplay=25, Front End=50), Alpha=$System.ContrailDiveAlpha)
      From=?
      To=?
      Angle Type=::NewEnumerator0
      Coordinate Space=::Local
      Normalize Vectors=true
     )
    )) {isSetParameters=true}
   /SetVariables(Particles.Position=&AddVectorToPosition(
     Position=$Particles.Position
     Vector=&Multiply_Vector(
      A=&TransformVector(Vector=$Engine.Owner.Velocity, Source Space=::World, Destination Space=::Local)
      B=&Multiply_VectorByFloat(
       Vector=<-1 -1 0>
       Float=&FloatFromCurve(
        FloatCurve=DataInterface() {propertyValues {Curve {keys [
            {arriveTangent=2.7932, interpMode=RCIM_Cubic, leaveTangent=2.7932, tangentMode=RCTM_User, time=0, value=0}
            {interpMode=RCIM_Cubic, tangentMode=RCTM_User, time=1, value=1}
           ]}}}
        CurveIndex=$Particles.PartitionParticles.Normalized Index In Partition
        Scale Curve=0.02
       )
      )
     )
    )) {isSetParameters=true}
   &ScaleMeshSize(Scale Factor=&MakeVector(X=1, Y=1, Z=&Lerp_Float(A=1, B=1.2, Alpha=$System.ContrailDiveAlpha)))
   &ScaleColor(
    Scale Alpha=&FNTimeOfDayScalar(Night Scale=0.5, Morning Scale=1, Evening Scale=1)
    ScaleA=true
    ScaleRGB=true
   )
   /SetVariables(Particles.SmoothVelocity:Vector3f=&Lerp_Vector(A=$Particles.SmoothVelocity, B=$Engine.Owner.Velocity, Alpha=0.1)) {isSetParameters=true}
   &SolveForcesAndVelocity(Clamp Velocity=false, Limit Acceleration=false)
   &DynamicMaterialParameters(
    Index 0 Param 1=&MergeHelper_FloatSwitch(0=0.75, 1=0.3333, 2=?, 3=?, Index=$Particles.PartitionParticles.Partition, Num=::NewEnumerator0)
    Index 0 Param 2=$Emitter.CustomTime
    Index 0 Param 3=&MergeHelper_FloatSwitch(0=0.025, 1=0.15, 2=?, 3=?, Index=$Particles.PartitionParticles.Partition, Num=::NewEnumerator0)
    Index 0 Param 4=$System.ContrailDiveAlpha
    Index 1 Vector 1:Vector3f=&TransformVector(
     Vector=&ReturnMeshOrientationAxis(Mesh Orientation Axis=::ZAxis, Mesh Orientation Quat=$Particles.MeshOrientation)
     Source Space=::World
     Destination Space=::Local
    )
    Module.Index0Float0Write=true
    Module.Index0Float1Write=true
    Module.Index0Float2Write=true
    Module.Index0Float3Write=true
    Module.Index1Float3Write=false
    Module.Index1VectorWrite=true
    Write Parameter Index 1=::NewEnumerator3
   )
  ]
  rendererClasses [&NiagaraMeshRendererProperties]
  renderers [
   {rendererClass=&NiagaraMeshRendererProperties, rendererIndex=0}
   {rendererClass=&NiagaraMeshRendererProperties, rendererIndex=1}
  ]
 }
 Spheres {
  emitterSpawnScript [/SetVariables() {isSetParameters=true}]
  emitterUpdateScript [
   &EmitterState(Life Cycle Mode=::NewEnumerator1, Loop Behavior=::NewEnumerator1, Loop Duration Mode=::NewEnumerator1, UseLoopDelay=false)
   &SpawnBurst_Instantaneous(Spawn Count=2, Use Spawn Probability=false, UseLoopCountLimit=false)
   /SetVariables(Emitter.CustomTime:NiagaraFloat=&Add_Float(A=$Emitter.CustomTime, B=&Lerp_Float(A=0.02, B=0.05, Alpha=$System.SmoothDiveAlpha))) {isSetParameters=true}
  ]
  particleSpawnScript [
   &PartitionParticles()
   &InitializeParticle(
    Color=<rgba 0.25 1 0.5625 0.5>
    Mesh Scale Mode=::NewEnumerator3
    Mesh Uniform Scale=5
    Sprite Size Mode=::NewEnumerator3
    Uniform Sprite Size=50
    UsePositionOffset=false
    Write Material Random=true
   )
  ]
  particleUpdateScript [
   &ParticleState(Kill Particles When Lifetime Has Elapsed=false, Let Infinitely Lived Particles Die When Emitter Deactivates=true)
   &ScaleSpriteSize(Uniform Curve Sprite Scale=DataInterface() {propertyValues {Curve {
       defaultValue=0
       keys [
        {arriveTangent=4.1785, interpMode=RCIM_Cubic, leaveTangent=4.1785, tangentMode=RCTM_User, time=0, value=0}
        {arriveTangent=1.1179, interpMode=RCIM_Cubic, leaveTangent=1.1179, tangentMode=RCTM_User, time=0.2042, value=0.6228}
        {arriveTangent=0.0224, interpMode=RCIM_Cubic, leaveTangent=0.0224, tangentMode=RCTM_User, time=1, value=1}
       ]
      }}})
   &ScaleMeshSize(Scale Factor=&VectorFromFloat(Value=&FloatFromCurve(FloatCurve=DataInterface() {propertyValues {Curve {
         defaultValue=0
         keys [
          {arriveTangent=4.1785, interpMode=RCIM_Cubic, leaveTangent=4.1785, tangentMode=RCTM_User, time=0, value=0}
          {arriveTangent=1.1179, interpMode=RCIM_Cubic, leaveTangent=1.1179, tangentMode=RCTM_User, time=0.2042, value=0.6228}
          {arriveTangent=0.0224, interpMode=RCIM_Cubic, leaveTangent=0.0224, tangentMode=RCTM_User, time=1, value=1}
         ]
        }}})))
   &SocketLocation(
    SamplePosOffset=true
    Sampled Position Offset=&Lerp_Vector(
     A=<0 15 -15>
     B=&MergeHelper_VectorSwitch(0=<15 10 25>, 1=<-15 10 25>, 2=?, 3=?, Index=$Particles.PartitionParticles.Partition, Num=::NewEnumerator0)
     Alpha=$System.ContrailDiveAlpha
    )
    Sampled Position Offset Coordinate Space:ENiagaraSkelMeshTransforms=::NewEnumerator2
    Socket Index=$Particles.PartitionParticles.Partition
    Socket Reader=$System.Socket Reader
    Socket Sampling Mode=::NewEnumerator3
    VelocitySpeedLimit=false
   )
   &SolveForcesAndVelocity(Clamp Velocity=false, Limit Acceleration=false)
   /SetVariables(Particles.DMP_01:Vector4f=&MakeVector4(X=2.5, Y=0, Z=0, W=0)) {isSetParameters=true}
   &DynamicMaterialParameters(
    Index 0 Param 1=$Emitter.CustomTime
    Index 0 Param 3=&MergeHelper_FloatSwitch(0=-1, 1=1, 2=?, 3=?, Index=$Particles.PartitionParticles.Partition, Num=::NewEnumerator0)
    Module.Index0Float0Write=true
    Module.Index0Float1Write=false
    Module.Index0Float2Write=true
    Module.Index0Float3Write=false
   )
   &FN_Light_Attributes(
    Color=<rgba 1.75 5 2.4872 1>
    Create Color Attribute=true
    Create Exponent Attribute=true
    Create Position Attribute=true
    Create Radius Attribute=true
    Exponent=7.5
    Position=&AddVectorToPosition(Position=$Output.SocketLocation.SampledPosition, Vector=<0 0 10>)
    Radius=60
    Scale Color=25
   )
   /SetVariables(Particles.MeshColor:LinearColor=&MultiplyLinearColor(
     Linear Color=<rgba 1 1 1 1>
     Scale Factor=&MakeVector4FromVectorAndFloat(
      "Vector (RGB)"=&FN_FrontEndVector3Modifier(Gameplay=<1 1 1>, Front End=<0.5 0.5 0.5>)
      "Float (Alpha)"=1
     )
    )) {isSetParameters=true}
   &ScaleColor(
    Scale Alpha=&FloatFromCurve(
     FloatCurve=DataInterface() {propertyValues {Curve {
        defaultValue=0
        keys [
         {arriveTangent=4.1785, interpMode=RCIM_Cubic, leaveTangent=4.1785, tangentMode=RCTM_User, time=0, value=0}
         {arriveTangent=1.1179, interpMode=RCIM_Cubic, leaveTangent=1.1179, tangentMode=RCTM_User, time=0.2042, value=0.6228}
         {arriveTangent=0.0224, interpMode=RCIM_Cubic, leaveTangent=0.0224, tangentMode=RCTM_User, time=1, value=1}
        ]
       }}}
     Scale Curve=0.5
    )
    ScaleA=true
    ScaleRGB=true
   )
  ]
  rendererClasses [
   &NiagaraMeshRendererProperties
   &NiagaraSpriteRendererProperties
   &NiagaraLightRendererProperties
  ]
  renderers [
   {rendererClass=&NiagaraMeshRendererProperties, rendererIndex=0}
   {rendererClass=&NiagaraSpriteRendererProperties, rendererIndex=1}
   {rendererClass=&NiagaraLightRendererProperties, rendererIndex=2}
  ]
 }
}
summary {
 systemName=NS_Contrail_SwoopClasp
 emitters [
  {
   emitterName=Spheres
   rendererClasses [
    &NiagaraMeshRendererProperties
    &NiagaraSpriteRendererProperties
    &NiagaraLightRendererProperties
   ]
  }
  {emitterName=Mesh_DNA, rendererClasses [&NiagaraMeshRendererProperties]}
  {emitterName=HUD, rendererClasses [&NiagaraSpriteRendererProperties]}
 ]
 userVariables [
  User.bIsFrontEnd {type=NiagaraBool}
  User.bIsFrontEndPreview {type=NiagaraBool}
  User.ParaGlideLeanAlpha {type=NiagaraFloat, defaultValue=0, description=""}
  User.RibbonIDOffset {type=NiagaraInt32}
 ]
}
